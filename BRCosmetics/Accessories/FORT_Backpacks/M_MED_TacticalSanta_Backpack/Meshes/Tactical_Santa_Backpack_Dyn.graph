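# Physics asset for the Tactical Santa backpack cosmetic, defining bodies and joints for its dangling gun, bow and grenade bones.
# The gun hangs from root with tight limits and carries a bow chain of base, two loops and two two-segment strings with looser swing.
# Two grenades also hang from root, each with a pin limited mostly to one swing axis.
# Capsules approximate most parts, while boxes shape the flat bow strings.
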PhysicsAsset /BRCosmetics/Accessories/FORT_Backpacks/M_MED_TacticalSanta_Backpack/Meshes/Tactical_Santa_Backpack_Dyn
bodies {
 dyn_bow_base [{shapeType=Capsule, length=2, radius=2, center=<0.2496 0 0>, rotation=<rot 90 0 0>}]
 dyn_bow_loop_l [{shapeType=Capsule, length=2, radius=2, center=<0 3.14 0.0001>, rotation=<rot 0 0 89.9994>}]
 dyn_bow_loop_r [{shapeType=Capsule, length=2, radius=2, center=<0 -3.0617 -0.0001>, rotation=<rot 0 0 89.9994>}]
 dyn_bow_string_l_01 [{shapeType=Box, extentX=7.5, extentY=2.5, extentZ=1, center=<0.3128 0 0>}]
 dyn_bow_string_l_02 [
  {shapeType=Box, extentX=7.5, extentY=3, extentZ=1, center=<1.1891 0 0.0001>}
  {shapeType=Box, extentX=3.0554, extentY=3.5376, extentZ=1, center=<6.0151 -0.8658 -0.257>, rotation=<rot -14.9999 0 0>}
 ]
 dyn_bow_string_r_01 [{shapeType=Box, extentX=4.9746, extentY=2.5, extentZ=1, center=<0.5005 0 -0.0003>}]
 dyn_bow_string_r_02 [
  {shapeType=Box, extentX=7.5, extentY=3, extentZ=1, center=<1.1891 0 0.0001>}
  {shapeType=Box, extentX=3.7728, extentY=3.5376, extentZ=1.7057, center=<5.8993 0.1472 -0.0138>, rotation=<rot -14.9999 0 0>}
 ]
 dyn_grenade_bott [{shapeType=Capsule, length=2, radius=2, center=<1.565 -0.4251 0>, rotation=<rot 90 0 0>}]
 dyn_grenade_pin_bott [{shapeType=Capsule, length=1.9255, radius=0.9258, center=<1.6651 0.0002 0>, rotation=<rot 90 0 0>}]
 dyn_grenade_pin_top [{shapeType=Capsule, length=2, radius=2, center=<1.5355 0 0>, rotation=<rot 90 0 0>}]
 dyn_grenade_top [{shapeType=Capsule, length=2, radius=2, center=<0.6164 0.0001 0>, rotation=<rot 90 0 0>}]
 dyn_gun [{shapeType=Capsule, length=44.9846, radius=4.4961, center=<-1.0846 -1.2604 2.1215>, rotation=<rot 0 62.6697 -87.8379>}]
 root [{shapeType=Capsule, length=2, radius=2}]
}
constraints [
 {bone1Name=dyn_gun, bone2Name=root, swing1LimitDegrees=5, swing2LimitDegrees=5, twistLimitDegrees=2}
 {bone1Name=dyn_bow_base, bone2Name=dyn_gun, swing1LimitDegrees=25, swing2LimitDegrees=25, twistLimitDegrees=10}
 {bone1Name=dyn_bow_loop_l, bone2Name=dyn_bow_base, swing1LimitDegrees=25, swing2LimitDegrees=2, twistLimitDegrees=25}
 {bone1Name=dyn_bow_loop_r, bone2Name=dyn_bow_base, swing1LimitDegrees=25, swing2LimitDegrees=2, twistLimitDegrees=25}
 {bone1Name=dyn_bow_string_r_01, bone2Name=dyn_bow_base, swing1LimitDegrees=45, swing2LimitDegrees=45, twistLimitDegrees=20}
 {bone1Name=dyn_bow_string_l_01, bone2Name=dyn_bow_base, swing1LimitDegrees=45, swing2LimitDegrees=45, twistLimitDegrees=10}
 {bone1Name=dyn_bow_string_l_02, bone2Name=dyn_bow_string_l_01, swing1LimitDegrees=45, swing2LimitDegrees=45, twistLimitDegrees=10}
 {bone1Name=dyn_grenade_top, bone2Name=root, swing1LimitDegrees=25, swing2LimitDegrees=25, twistLimitDegrees=10}
 {bone1Name=dyn_grenade_pin_top, bone2Name=dyn_grenade_top, swing1LimitDegrees=2, swing2LimitDegrees=25, twistLimitDegrees=2}
 {bone1Name=dyn_grenade_bott, bone2Name=root, swing1LimitDegrees=25, swing2LimitDegrees=25, twistLimitDegrees=10}
 {bone1Name=dyn_grenade_pin_bott, bone2Name=dyn_grenade_bott, swing1LimitDegrees=2, swing2LimitDegrees=25, twistLimitDegrees=2}
 {bone1Name=dyn_bow_string_r_02, bone2Name=dyn_bow_string_r_01, swing1LimitDegrees=45, swing2LimitDegrees=45, twistLimitDegrees=20}
]
