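# Animation Blueprint for the Tactical Santa backpack cosmetic, adding physics-driven secondary motion to its skeleton.
# The AnimGraph feeds the local reference pose through a RigidBody node using the Tactical_Santa_Backpack_Dyn physics asset, then outputs the result.
# Simulation runs in root-bone space with overridden gravity, passing component velocity and clamped acceleration into the bodies.
# The EventGraph holds only an unconnected update event, with no variables or logic.
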AnimBlueprint /BRCosmetics/Accessories/FORT_Backpacks/M_MED_TacticalSanta_Backpack/Meshes/Tactical_Santa_Backpack_AnimBP
target_skeleton=&/Tactical_Santa_Backpack_Skeleton
graphs [
 AnimGraph {nodes [
   n1 = RigidBody(ComponentPose=Local To Component(LocalPose=Local Space Ref Pose())) {
    bOverrideWorldGravity=true
    overridePhysicsAsset=&/Tactical_Santa_Backpack_Dyn
    simulationSpace=BaseBoneSpace
    alphaBoolBlend {blendOption=Linear}
    baseBoneRef {boneName=root}
    componentAppliedLinearAccClamp=<1000 1000 1000>
    componentLinearAccScale=<1 1 1>
    componentLinearVelScale=<1 1 1>
    overrideWorldGravity {z=-980}
   }
   Output Pose(Result=Component To Local(ComponentPose=@n1)) {layerGroup=None}
  ]}
 EventGraph {nodes [Event BlueprintUpdateAnimation()]}
]
variables []
